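; Player pawn for ElbowChat that stops colliding with the throne and shield and gets placed at the throne.
; On BeginPlay it looks up the throne by class, using a looping timer to retry when missing.
; Found actors are ignored by capsule and mesh movement, then the authority moves the pawn to the throne plus offsets.
; A ShieldAdded gameplay event sets ShieldObject, whose OnRep also ignores shield collision.
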(event Custom|TryGetThroneActor
  (if (Variables|Default|GetCanCollideWithThrone)
    _
    (elif (Variables|Default|GetFoundThrone)
      _
      (else
        (bind (_output_get _outputpin) (CallFunction|GetThroneReference))
        (if (not _output_get)
          (CallFunction|SetandRetryTimerHandle (Variables|Default|GetHNDL_FindThroneTimer) (EventDispatchers|CreateEvent :!event Custom|TryGetThroneActor))))))
  (if (Utilities|Time|IsValidTimerHandle (Variables|Default|GetHNDL_FindThroneTimer))
    (Utilities|Time|ClearandInvalidateTimerbyHandle (Variables|Default|GetHNDL_FindThroneTimer)))
  (Utilities|Array|Add (Variables|Default|GetActorsToRemoveCollisionsWith) _outputpin)
  (CallFunction|DisableCollidingwithActors))

(event Custom|TryDisableCollisionsWithActors
  (if (Variables|Default|GetCanCollideWithThrone)
    _
    (elif (Variables|Default|GetFoundThrone)
      _
      (else
        (bind _output_get (CallFunction|GetThroneReference))
        (if _output_get
          (if (Utilities|Time|IsValidTimerHandle (Variables|Default|GetHNDL_FindThroneTimer))
            (Utilities|Time|ClearandInvalidateTimerbyHandle (Variables|Default|GetHNDL_FindThroneTimer)))
          (CallFunction|TryDisableCollisionswithActors)
          (else
            (CallFunction|SetandRetryTimerHandle (Variables|Default|GetHNDL_FindThroneTimer) :!label _label))))))
  (if (Variables|Default|GetCanCollideWithShield)
    _
    (elif (Variables|Default|GetFoundShield)
      _
      (else
        (bind _foundshield (CallFunction|GetShieldReference))
        (if (not _foundshield)
          (goto _label)))))
  (CallFunction|DisableCollidingwithActors))

(event Custom|DisableCollidingWithActors
  (for _array_element (Variables|Default|GetActorsToRemoveCollisionsWith)
    (Collision|IgnoreActorwhenMoving (Variables|Character|GetCapsuleComponent) _array_element true)
    (Collision|IgnoreActorwhenMoving (Variables|Character|GetMesh) _array_element true))
  (CallFunction|MovePawntoThrone))

(event Custom|MovePawnToThrone
  (bind _throneactor (Variables|Default|GetThroneActor))
  (Utilities|IsValid _throneactor
    (:Is%20Valid
      (Networking|SwitchHasAuthority
        (:Authority
          (Transformation|SetActorLocation self (Math|Vector|MakeVector (.x (Transformation|GetActorLocation _throneactor)) (+ (.y (Transformation|GetActorLocation _throneactor)) (Variables|Default|GetthroneForwardOffset)) (+ (.z (Transformation|GetActorLocation _throneactor)) (+ (Components|Capsule|GetUnscaledCapsuleHalfHeight (Variables|Character|GetCapsuleComponent)) (Variables|Default|GetthroneUpOffset)))) false true))))))

(event EventBeginPlay
  (|CallParentFunction)
  (CallFunction|TryGetThroneActor)
  (bind (_asynctaskproxy _payload) (Ability|Async|WaitGameplayEventToActor 0 (Variables|Default|GetTag_ShieldAdded) true)
    (:EventReceived
      (Variables|Default|SetShieldObject (.OptionalObject (Utilities|Struct|BreakGameplayEventData _payload))))))

(event Collision|EventActorBeginOverlap
  (params (OtherActor :type Actor%20Object%20Reference))
  (|CallParentFunction OtherActor) ; This node is disabled and will not be called.\nDrag off pins to build functionality.
)

(event EventTick
  (params (DeltaSeconds :type Float%20%28single-precision%29))
  (|CallParentFunction DeltaSeconds) ; This node is disabled and will not be called.\nDrag off pins to build functionality.
)

(fn GetActorReference
  (params (ActorClass :type Actor%20Class%20Reference))
  (bind _outactors (Actor|GetAllActorsOfClass ActorClass))
  (if (> (Utilities|Array|Length _outactors) 0)
    (bind _result (ArrayMacros|SafeArrayGet _outactors)
      (:%20%20
        (return _result)))))

(fn GetShieldReference
  (bind _actorfound (CallFunction|GetActorReference "/ElbowChatNPC/Assets/BGA_ElbowChat_Shield.BGA_ElbowChat_Shield_C"))
  (bind _asbga_elbow_chat_shield (Utilities|Casting|CastToBGA_ElbowChat_Shield _actorfound)
    (:then
      (bind _output_get (Variables|Default|SetShieldActor _asbga_elbow_chat_shield))
      (Variables|Default|SetFoundShield true)
      (Utilities|Array|Add (Variables|Default|GetActorsToRemoveCollisionsWith) _output_get))
    (:CastFailed
      (Variables|Default|SetFoundShield false)))
  (return (Variables|Default|GetFoundShield) _output_get))

(fn GetThroneReference
  (bind _actorfound (CallFunction|GetActorReference (Variables|Default|GetthroneClass)))
  (bind _asbp_elbow_chat_throne_a (Utilities|Casting|CastToBP_ElbowChat_Throne_A _actorfound)
    (:then
      (bind _output_get (Variables|Default|SetThroneActor _asbp_elbow_chat_throne_a))
      (bind _output_get_2 (Variables|Default|SetFoundThrone true))
      (return _output_get_2 _output_get))
    (:CastFailed
      (bind _output_get_3 (Variables|Default|SetFoundThrone false))
      (return _output_get_3 _output_get))))

(fn OnRep_ShieldObject
  (bind _shieldobject (Variables|Default|GetShieldObject))
  (Utilities|IsValid _shieldobject
    (:Is%20Valid
      (bind _asactor (Utilities|Casting|CastToActor _shieldobject))
      (Collision|IgnoreActorwhenMoving (Variables|Character|GetCapsuleComponent) _asactor true)
      (Collision|IgnoreActorwhenMoving (Variables|Character|GetMesh) _asactor true))))

(fn SetAndRetryTimerHandle
  (params (TimerHandle :type Timer%20Handle%20Structure) (Event :type Delegate))
  (if (not (Utilities|Time|IsValidTimerHandle TimerHandle))
    (bind _returnvalue (Utilities|Time|SetTimerbyEvent Event 1.0 true))
    (Variables|Default|SetHNDL_FindThroneTimer _returnvalue)))

(fn ConstructionScript
  (|CallParentFunction))
